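# Physics Asset for the BR variant of the Orion vehicle body mesh in the vehicle cosmetics content.
# It holds a single body on the root bone, with no collision shapes listed.
# It has no constraints, so no joints or ragdoll chains are set up.
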PhysicsAsset /VehicleCosmetics/Bodies/Orion/Meshes/Body_Orion_Physics_BR
bodies {root []}
constraints []
